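# Physics asset for the Peely Mech medium male player body, covering only its left and right dangling hose bones.
# Each side has three capsule bodies forming a chain from hose base through bottom segment to mid segment.
# Angular limits keep the hoses fairly stiff, with tight twist throughout and somewhat wider swing at the bottom-to-mid joint than at the base.
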PhysicsAsset /Game/Characters/Player/Male/Medium/Bodies/M_MED_PeelyMech/M_MED_Peely_Mech_Physics
bodies {
 dyn_hoseBottom_Base_l [{shapeType=Capsule, length=0.505, radius=0.505, rotation=<rot 0 -90 0>}]
 dyn_hoseBottom_Base_r [{shapeType=Capsule, length=0.505, radius=0.505, rotation=<rot 0 -90 0>}]
 dyn_hoseBottom_l [{shapeType=Capsule, length=9.5047, radius=1.3274, center=<-4.3473 -0.0001 0>, rotation=<rot 90 0 0>}]
 dyn_hoseBottom_r [{shapeType=Capsule, length=8.0941, radius=1.4262, center=<-3.9615 0.0002 0>, rotation=<rot 90 0 0>}]
 dyn_hoseMid_l [{shapeType=Capsule, length=5.5357, radius=1.2918, center=<-2.0223 0.0122 0.0074>, rotation=<rot 0 89.9111 -89.371>}]
 dyn_hoseMid_r [{shapeType=Capsule, length=5.5322, radius=1.2843, center=<-2.0112 0.0183 -0.0241>, rotation=<rot 0 89.1392 -89.7517>}]
}
constraints [
 {bone1Name=dyn_hoseBottom_r, bone2Name=dyn_hoseBottom_Base_r, swing1LimitDegrees=10, swing2LimitDegrees=10, twistLimitDegrees=6}
 {bone1Name=dyn_hoseBottom_l, bone2Name=dyn_hoseBottom_Base_l, swing1LimitDegrees=10, swing2LimitDegrees=10, twistLimitDegrees=6}
 {bone1Name=dyn_hoseMid_r, bone2Name=dyn_hoseBottom_r, swing1LimitDegrees=18, swing2LimitDegrees=18, twistLimitDegrees=6}
 {bone1Name=dyn_hoseMid_l, bone2Name=dyn_hoseBottom_l, swing1LimitDegrees=18, swing2LimitDegrees=18, twistLimitDegrees=6}
]
